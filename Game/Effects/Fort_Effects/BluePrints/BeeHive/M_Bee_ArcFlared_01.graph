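# Translucent, two-sided lit material for mesh particles in the BeeHive effects, apparently a flared arc trail.
# Two panners, timed by the PannerSpeed dynamic parameter, scroll infected-pattern color and normal maps, which are blended.
# The averaged color, tinted by particle color, feeds an optional emissive through an adjustment function.
# A separately scrolling trail mask, double-sided Fresnel from the blended normals, and particle and vertex alpha multiply into a fade mask.
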Material /Game/Effects/Fort_Effects/BluePrints/BeeHive/M_Bee_ArcFlared_01
graph {
 properties {
  bEnableResponsiveAA=true
  bUsedWithMeshParticles=true
  bUsedWithNiagaraMeshParticles=true
  blendMode=BLEND_Translucent
  pixelDepthOffsetMode=PDOM_Legacy
  translucencyLightingMode=TLM_VolumetricPerVertexDirectional
  twoSided=true
  usedShadingModels=Default Lit
 }
 nodes [
  n1 = ScalarParameter() {defaultValue=0.385, parameterName=Ua}
  n2 = ScalarParameter() {defaultValue=0.34, parameterName=Va}
  n3 = Multiply(A=Time(), B=DynamicParameter() {paramNames [PannerSpeed,Param2,Param3,Param4]}.PannerSpeed)
  n4 = Panner(Coordinate=Multiply(A=TextureCoordinate(), B=AppendVector(A=@n1, B=@n2)), Time=@n3) {bFractionalPart=true, speedX=0.763, speedY=0.0712}
  n5 = ScalarParameter() {defaultValue=0.325, parameterName=Ub}
  n6 = ScalarParameter() {defaultValue=0.427, parameterName=Vb}
  n7 = Panner(Coordinate=Multiply(A=TextureCoordinate(), B=AppendVector(A=@n5, B=@n6)), Time=@n3) {bFractionalPart=true, speedX=0.6096, speedY=-0.0598}
  n8 = Divide(A=Add(
    A=TextureSample(UVs=@n4) {texture=&2ndInfectedMapLinear}.RGB
    B=TextureSample(UVs=@n7) {texture=&2ndInfectedMapLinear}.RGB
   ))
  n9 = ParticleColor()
  MaterialFunctionCall(Emissive=StaticSwitchParameter(True=Multiply(A=@n8, B=@n9.RGB), False=Constant()) {parameterName=UseEmissive}) {materialFunction=&MF_FX_EmissiveAdjustments}
  n10 = TextureSample(UVs=Panner(Coordinate=TextureCoordinate() {uTiling=2}) {bFractionalPart=true, speedX=1.221}) {texture=/Game/Effects/Fort_Effects/Textures/Tile/T_Trail_Masks_Tiled}
  n11 = Normalize(VectorInput=Add(
    A=TextureSample(UVs=@n4) {samplerType=SAMPLERTYPE_Normal, texture=/Game/Global/Textures/Default/Tileable_Patterns/2ndInfectedMapNormal}.RGB
    B=TextureSample(UVs=@n7) {samplerType=SAMPLERTYPE_Normal, texture=&2ndInfectedMapNormal}.RGB
   ))
  n12 = MaterialFunctionCall(Exponent=ScalarParameter() {defaultValue=4, parameterName=Fresnel Exponent}, Vertex / Pixel Normal=Transform(None=@n11)) {materialFunction=&MF_Fresnel_FX}
  Multiply(A=@n9.A, B=Multiply(A=Multiply(A=@n10.G, B=@n12.Fresnel Double Sided), B=VertexColor().A))
  Multiply(A=@n11, B=VertexNormalWS())
 ]
}
