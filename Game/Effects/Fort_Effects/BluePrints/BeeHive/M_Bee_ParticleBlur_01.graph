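# Translucent lit particle-sprite material for bees in the BeeHive effect, with responsive AA and a motion-blurred look at higher speeds.
# Particle speed blends opacity from the texture's G channel toward its scaled alpha; the matching color blend uses R on both sides, so has no effect.
# A sine flicker at a random per-particle rate modulates opacity where the mask texture allows.
# Particle color tints and fades the result, and MF_Distance_SyncedFade_FX adds distance fading.
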Material /Game/Effects/Fort_Effects/BluePrints/BeeHive/M_Bee_ParticleBlur_01
graph {
 properties {
  bEnableResponsiveAA=true
  bUsedWithParticleSprites=true
  blendMode=BLEND_Translucent
  pixelDepthOffsetMode=PDOM_Legacy
  translucencyLightingMode=TLM_VolumetricPerVertexDirectional
  usedShadingModels=Default Lit
 }
 nodes [
  n1 = ParticleColor()
  n2 = TextureSample() {texture=&/T_Bee_ParticleBlur_01}
  n3 = Clamp(None=Multiply(A=ParticleSpeed(), B=ScalarParameter() {defaultValue=0.003, parameterName=Speed Lerp Modulate}))
  n4 = LinearInterpolate(A=@n2.G, B=Multiply(A=@n2.A, B=ScalarParameter() {defaultValue=0.6, parameterName=Fast Opacity}), Alpha=@n3)
  n5 = Multiply(
   A=Time()
   B=LinearInterpolate(
    A=ScalarParameter() {defaultValue=0.8, parameterName=Time Min}
    B=ScalarParameter() {defaultValue=2, parameterName=Time Max}
    Alpha=ParticleRandom()
   )
  )
  n6 = Multiply(A=@n4, B=ConstantBiasScale(None=Sine(None=@n5) {period=0.11}))
  n7 = TextureSample() {texture=&/T_Bee_ParticleBlurMask_01}
  MaterialFunctionCall(
   Opacity In=Multiply(A=@n1.A, B=LinearInterpolate(A=@n4, B=@n6, Alpha=@n7.R))
   Position=ParticlePositionWS()
   Fade_Start_Distance=ScalarParameter() {defaultValue=200, parameterName=Start Fade}
   Fade_End_Distance=ScalarParameter() {defaultValue=100, parameterName=End Fade}
  ) {materialFunction=&MF_Distance_SyncedFade_FX}
  Multiply(A=@n1.RGB, B=LinearInterpolate(A=@n2.R, B=@n2.R, Alpha=@n3))
 ]
}
